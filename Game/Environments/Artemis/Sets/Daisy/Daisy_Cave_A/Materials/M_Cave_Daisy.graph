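# Cave set material for Daisy_Cave_A that layers sand, gravel, world-aligned rock and grass, outputting world-space normals.
# Vertex colors, height lerps and surface normals mask the layers, with separate tints for path, cave and outside sand.
# Epic quality adds wet walls inside hard-coded world-position sphere masks, while forward and mobile paths simplify textures.
# The base blends into the landscape runtime virtual texture by height above it, unless NoRVT, forward or mobile bypasses this.
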Material /Game/Environments/Artemis/Sets/Daisy/Daisy_Cave_A/Materials/M_Cave_Daisy
graph {
 properties {
  bTangentSpaceNormal=false
  bUseMaterialAttributes=true
  ditheredLODTransition=true
  pixelDepthOffsetMode=PDOM_Legacy
  usedShadingModels=Default Lit
 }
 nodes [
  ComponentMask() {r=true}
  n1 = Multiply(A=TextureCoordinate(), B=ScalarParameter() {defaultValue=40, group=Sand, parameterName=Gravel Scale})
  n2 = TextureSample(
   UVs=@n1
   Tex=TextureObjectParameter() {group=Sand, parameterName=Gravel Diffuse, texture=/Game/Environments/Apollo/Sets/Terrain/Textures/T_Apollo_Scree_D_01}
  )
  ComponentMask(None=@n2.R) {a=true}
  n3 = ScalarParameter() {defaultValue=0.25, parameterName=Cave Center Sand Contrast}
  Composite!(Hardness=@n3) {subgraphName=Sand Location}
  Composite!() {subgraphName=Sand To Rock}
  Composite() {subgraphName=Grass Mask}
  n4 = NamedRerouteUsage()
  Composite!(Alpha2=@n4)
  n5 = NamedRerouteUsage()
  Composite(Alpha2=@n5)
  n6 = Multiply(A=TextureCoordinate(), B=ScalarParameter() {defaultValue=40, group=Sand, parameterName=Sand Scale})
  n7 = TextureObjectParameter() {
   group=Sand
   parameterName=Sand Specular
   samplerType=SAMPLERTYPE_Masks
   texture=/Game/Environments/Apollo/Sets/Terrain/Textures/T_Apollo_Sand_S_01
  }
  n8 = Constant() {r=0.5}
  n9 = MaterialFunctionCall(Float3=ShadingPathSwitch(Default=TextureSample(UVs=@n6, Tex=@n7) {samplerType=SAMPLERTYPE_Masks}.RGB, Forward=@n8, Mobile=@n8)) {materialFunction=&BreakOutFloat3Components}
  NamedRerouteDeclaration(None=@n9.B) {name=Sand R, nodeColor=<rgba 0.4531 0 1 1>}
  NamedRerouteDeclaration(None=@n9.R) {name=Sand S, nodeColor=<rgba 0.9219 0 1 1>}
  n10 = TextureSample(
   UVs=@n6
   Tex=TextureObjectParameter() {group=Sand, parameterName=Sand Diffuse, texture=/Game/Environments/Apollo/Sets/Terrain/Textures/T_Apollo_Sand_D_01}
  )
  NamedRerouteDeclaration(None=Multiply(
    A=NamedRerouteDeclaration(None=@n10.R) {name=Sand D, nodeColor=<rgba 0.9922 0 1 1>}
    B=VectorParameter() {group=Sand, parameterName=Sand Path, sortPriority=-4, defaultValue=<rgba 0.4896 0.4896 0.4896 0>}.RGB
   )) {name=Sand Color, nodeColor=<rgba 0.9922 0 1 1>}
  NamedRerouteDeclaration(None=@n2.A) {name=Gravel A, nodeColor=<rgba 1 0.6562 0 1>}
  NamedRerouteDeclaration(None=ComponentMask(None=@n10.A) {a=true}) {name=Sand A, nodeColor=<rgba 1 0.0469 0 1>}
  n11 = VectorParameter() {group=Sand, parameterName=Sand Outside Path, sortPriority=-6, defaultValue=<rgba 0.5434 0.4605 0.2481 0.5>}
  n12 = VectorParameter() {group=Sand, parameterName=Sand Outside, sortPriority=-7, defaultValue=<rgba 0.3038 0.2454 0.0956 0.5>}
  n13 = Multiply(A=LinearInterpolate(A=@n11.RGB, B=@n12.RGB, Alpha=NamedRerouteUsage()), B=NamedRerouteUsage())
  n14 = NamedRerouteUsage()
  n15 = NamedRerouteUsage()
  n16 = Constant3Vector() {constant=<rgba 0.25 0.25 1 1>}
  n17 = SetMaterialAttributes!(
   MaterialAttributes
   Base Color=NamedRerouteUsage()
   Roughness=NamedRerouteUsage()
   Specular=NamedRerouteUsage()
   Normal=NamedRerouteDeclaration(None=Multiply(A=@n15, B=@n16)) {name=Sand Path N, nodeColor=<rgba 1 0.5156 0 1>}
  ) {
   desc=Sand Path
   attributeSetTypes [
    #1
    #2
    #3
    #4
   ]
  }
  n18 = VectorParameter() {group=Sand, parameterName=Sand Cave, sortPriority=-5, defaultValue=<rgba 0.2173 0.3483 0.7483 0.5>}
  n19 = Multiply(A=Multiply(A=NamedRerouteUsage(), B=@n18.RGB), B=NamedRerouteUsage())
  n20 = NamedRerouteUsage()
  MaterialExpressionReroute_14 = Reroute(None=@n4)
  n21 = MaterialFunctionCall(
   Transition Phase=VertexColor().R
   Height Texture=LinearInterpolate(Alpha=@MaterialExpressionReroute_14.Alpha2) {constA=-1, constB=1.4}
   Contrast=Constant() {r=0.25}
  ) {materialFunction=&HeightLerp}
  MaterialExpressionReroute_0 = Reroute(None=@n21.Alpha)
  n22 = NamedRerouteDeclaration(None=BlendMaterialAttributes(
    A=@n17
    B=SetMaterialAttributes!(MaterialAttributes=@n17, Base Color=@n19, Normal=@n20) {desc=Sand, attributeSetTypes [#1, #4]}
    Alpha=NamedRerouteDeclaration(None=@MaterialExpressionReroute_0.Alpha) {name=Path Mask, nodeColor=<rgba 1 0.8672 0 1>}
   )) {name=Sand Mix, nodeColor=<rgba 1 0.2344 0 1>}
  n23 = MaterialFunctionCall(Transition Phase=VertexColor().B, Height Texture=NamedRerouteUsage()) {materialFunction=&HeightLerp}
  MaterialExpressionReroute_2 = Reroute(None=@n23.Alpha)
  NamedRerouteDeclaration(None=BlendMaterialAttributes(
    A=SetMaterialAttributes(MaterialAttributes=@n22, Base Color=@n13, Normal=@n14) {desc=Sand Outside, attributeSetTypes [#1, #4]}
    B=@n22
    Alpha=NamedRerouteDeclaration(None=@MaterialExpressionReroute_2.Alpha) {name=Outside Sand Mask, nodeColor=<rgba 1 0.6562 0 1>}
   )) {name=WS Sand Mat, nodeColor=<rgba 0 1 0.25 1>}
  n24 = ComponentMask(None=WorldPosition().XYZ) {b=true}
  n25 = Subtract(A=@n24, B=Constant() {r=2937.0996})
  n26 = ScalarParameter() {defaultValue=-1000, group=Grass, parameterName=Top Soft Mask}
  NamedRerouteDeclaration(None=Saturate(None=Divide(A=@n25, B=@n26))) {name=WS Top Part soft mask, nodeColor=<rgba 0 1 0.25 1>}
  n27 = MaterialFunctionCall!(
   Diffuse Texture Object=TextureObject() {texture=/Game/Environments/Artemis/Sets/HollowBase/Terrain/Textures/T_Artemis_Cliff_HollowBase_01_D}
   Spec Texture Object
   TextureSize=Multiply(A=ScalarParameter() {defaultValue=3, group=Rocks, parameterName=ScaleRocks}, B=Constant() {r=1500})
   --------------------------
   Use seperate settings for Z projection
   Use seperate texture for Z projection
   Texture Size Z projection
   Z Plane Diffuse Texture Object
   Z Plane Spec Texture Object
   "--------------------------      "
   ProjectionTransitionContrast
   Use Texture Alpha Heightmap to Refine the Transistion
   HeightLerpTransitionContrast
   PickTheDiffuseHeightChannel
   TheDiffuseHeightChannel
   "--------------------------       "
   NormalTextureObject=TextureObject() {samplerType=SAMPLERTYPE_Normal, texture=/Game/Environments/Artemis/Sets/Terrain/Textures/T_Artemis_Cliff_01_N}
   World Space Normals=StaticBool()
   "--------------------------  "
   Export Float 4 Diffuse=StaticBool()
   Export Float 4 Spec
   "--------------------------     "
   WorldPosition
   World Space Normal
  ) {materialFunction=/Engine/Functions/Engine_MaterialFunctions02/Texturing/WorldAlignedTextures_Complex}
  n28 = ComponentMask(None=@n27.XYZ Diffuse Texture) {b=true, g=true, r=true}
  n29 = Constant() {r=0.5}
  n30 = Constant() {r=0.6393}
  n31 = Constant() {r=2}
  n32 = ComponentMask(None=@n28) {r=true}
  n33 = Constant3Vector() {constant=<rgba 0 0 1 0>}
  NamedRerouteDeclaration(None=MakeMaterialAttributes!(
    BaseColor=@n28
    Metallic
    Specular=Multiply(A=@n28, B=@n29)
    Roughness=LinearInterpolate(A=@n30, B=@n31, Alpha=@n32)
    Anisotropy
    EmissiveColor
    Opacity
    OpacityMask
    Normal=ShadingPathSwitch(Default=@n27.Normal XYZ Texture, Forward=@n33, Mobile=@n33)
    Tangent
    WorldPositionOffset
    SubsurfaceColor
    ClearCoat
    ClearCoatRoughness
    AmbientOcclusion
    Refraction
    CustomizedUVs_0
    CustomizedUVs_1
    CustomizedUVs_2
    CustomizedUVs_3
    CustomizedUVs_4
    CustomizedUVs_5
    CustomizedUVs_6
    CustomizedUVs_7
    PixelDepthOffset
    ShadingModel
    Displacement
   )) {name=WS Rocks, nodeColor=<rgba 1 0.1172 0 1>}
  n34 = MaterialFunctionCall!(CombinedLandscapeLayersMask, BypassUseRVT, BypassBiomeColors) {materialFunction=/Game/Athena/Apollo/Environments/Landscape/Materials/Biome/CommonMaterials/Mobile/MF_Mobile_Common_Grass}
  n35 = GetMaterialAttributes(None=NamedRerouteDeclaration(None=@n34.ForStaticMesh) {name=BlendedLayers, nodeColor=<rgba 1 0 0.9844 1>}) {attributeGetTypes [#4]}
  NamedRerouteDeclaration(None=SetMaterialAttributes!(
    MaterialAttributes=@n35.MaterialAttributes
    Normal=Transform(None=@n35.Normal) {transformSourceType=TRANSFORMSOURCE_World, transformType=TRANSFORM_Tangent}
   ) {attributeSetTypes [#4]}) {name=Grass, nodeColor=<rgba 0 0.7656 1 1>}
  n36 = GetMaterialAttributes(None=NamedRerouteUsage()) {attributeGetTypes [#1, #2]}
  n37 = TextureSample(
   UVs=Multiply(A=TextureCoordinate(), B=ScalarParameter() {defaultValue=25, group=Rocks, parameterName=Details Tile})
   Tex=TextureObject() {texture=/Game/Environments/Apollo/Rocks/Natural_Rocks/Textures/T_RockDetails_02D}
  )
  n38 = SetMaterialAttributes!(
   MaterialAttributes=@n36.MaterialAttributes
   Base Color=MaterialFunctionCall(Base=@n36.Base Color, Blend=NamedRerouteDeclaration(None=@n37.R) {name=Details, nodeColor=<rgba 0.5 1 0 1>}) {materialFunction=&Blend_Overlay}.Result
   Roughness=@n36.MaterialAttributes
  ) {attributeSetTypes [#1, #2]}
  n39 = SetMaterialAttributes!(
   MaterialAttributes=@n36.MaterialAttributes
   Base Color=MaterialFunctionCall(Base=@n36.Base Color, Blend=Constant() {r=0.5}) {materialFunction=&Blend_Overlay}.Result
  ) {attributeSetTypes [#1]}
  n40 = NamedRerouteUsage()
  n41 = GetMaterialAttributes(None=ShadingPathSwitch(Default=QualitySwitch(Default=@n38, Low=@n39), Forward=@n40, Mobile=@n40)) {attributeGetTypes [#1]}
  n42 = VectorParameter() {group=Rocks, parameterName=QuickTone, defaultValue=<rgba 0.36 0.3426 0.3212 0>}
  NamedRerouteDeclaration(None=SetMaterialAttributes(MaterialAttributes=@n41.MaterialAttributes, Base Color=Multiply(A=@n41.Base Color, B=@n42.RGB)) {attributeSetTypes [#1]}) {name=Textures + Details, nodeColor=<rgba 1 0.8203 0 1>}
  n43 = VectorParameter() {group=Sand, parameterName=GravelColor Outside, sortPriority=-8, defaultValue=<rgba 0.5498 0.8594 0.8049 0.5>}
  n44 = VectorParameter() {group=Sand, parameterName=GravelColor Cave, sortPriority=-3, defaultValue=<rgba 0.8594 0.8489 0.6632 0.5>}
  n45 = LinearInterpolate(A=@n43.RGB, B=@n44.RGB, Alpha=NamedRerouteUsage())
  n46 = TextureObjectParameter() {
   group=Sand
   parameterName=Gravel Normal
   samplerType=SAMPLERTYPE_Normal
   texture=/Game/Environments/Apollo/Sets/Terrain/Textures/T_Apollo_Scree_N_01
  }
  n47 = TextureSample(UVs=@n1, Tex=@n46)
  n48 = Constant3Vector() {constant=<rgba 0 0 1 0>}
  n49 = SetMaterialAttributes!(
   MaterialAttributes
   Base Color=NamedRerouteDeclaration(None=Multiply(A=@n45, B=@n2.R)) {name=Gravel D, nodeColor=<rgba 1 0.6563 0 1>}
   Roughness=NamedRerouteDeclaration(None=Constant() {r=0.55}) {name=Gravel R, nodeColor=<rgba 1 0.6562 0 1>}
   Specular=NamedRerouteDeclaration(None=Constant() {r=0.35}) {name=Gravel S, nodeColor=<rgba 1 0.6562 0 1>}
   Normal=NamedRerouteDeclaration(None=ShadingPathSwitch(Default=@n47.RGB, Forward=@n48, Mobile=@n48)) {name=Gravel N, nodeColor=<rgba 1 0.6562 0 1>}
  ) {attributeSetTypes [
    #1
    #2
    #3
    #4
   ]}
  NamedRerouteDeclaration(None=@n49) {name=Gravel, nodeColor=<rgba 0.9453 0 1 1>}
  n50 = VertexColor()
  MaterialExpressionReroute_38 = Reroute(None=@n5)
  n51 = MaterialFunctionCall(
   Transition Phase=OneMinus(None=@n50.G)
   Height Texture=LinearInterpolate(Alpha=@MaterialExpressionReroute_38.Alpha2) {constA=-1}
   Contrast=Multiply(A=@n50.G, B=Constant() {r=1.5})
  ) {materialFunction=&HeightLerp}
  MaterialExpressionReroute_37 = Reroute(None=@n51.Alpha)
  NamedRerouteDeclaration(None=@MaterialExpressionReroute_37.Alpha) {name=Gravel Height, nodeColor=<rgba 0 1 0.8125 1>}
  NamedRerouteDeclaration(None=@n34.Height) {name=Grass H, nodeColor=<rgba 0 0.7187 1 1>}
  n52 = TextureObjectParameter() {
   group=Sand
   parameterName=Sand Normal
   samplerType=SAMPLERTYPE_Normal
   texture=/Game/Environments/Apollo/Sets/Terrain/Textures/T_Apollo_Sand_N_01
  }
  n53 = Constant3Vector() {constant=<rgba 0 0 1 0>}
  NamedRerouteDeclaration(None=ShadingPathSwitch(Default=TextureSample(UVs=@n6, Tex=@n52) {samplerType=SAMPLERTYPE_Normal}.RGB, Forward=@n53, Mobile=@n53)) {name="Sand  N Z", nodeColor=<rgba 1 0 0.2578 1>}
  OneMinus()
  PinBase() {reroutePins [Hardness {expression=&self:MaterialExpressionReroute_4}]}
  PinBase!(Alpha=@n23.Alpha) {reroutePins [Alpha {expression=&self:MaterialExpressionReroute_2}]}
  PinBase!() {reroutePins [Alpha2 {expression=&self:MaterialExpressionReroute_14}]}
  PinBase!(Alpha=@n21.Alpha, Lerp Alpha No Contrast=@n21.Alpha) {reroutePins [
    Alpha {expression=&self:MaterialExpressionReroute_0}
    Lerp Alpha No Contrast {expression=&self:MaterialExpressionReroute_11}
   ]}
  PinBase!() {reroutePins [Alpha2 {expression=&self:MaterialExpressionReroute_38}]}
  PinBase!(Alpha=@n51.Alpha) {reroutePins [Alpha {expression=&self:MaterialExpressionReroute_37}]}
  PinBase!()
  n54 = VertexNormalWS()
  n55 = VertexColor()
  n56 = NamedRerouteUsage()
  n57 = Constant() {r=-6.5504}
  n58 = Constant() {r=1}
  n59 = MaterialFunctionCall(
   Transition Phase=Multiply(A=ComponentMask(None=@n54) {b=true}, B=@n55.A)
   Height Texture=LinearInterpolate(A=@n56, B=LinearInterpolate(A=@n57, B=@n58, Alpha=@n56), Alpha=@n55.A)
   Contrast=Multiply(A=Constant() {r=0.5}, B=@n55.A)
  ) {materialFunction=&HeightLerp}
  PinBase!(Alpha=@n59.Alpha) {reroutePins [Alpha {expression=&self:MaterialExpressionReroute_9}]}
  PinBase!()
  n60 = VertexNormalWS()
  n61 = MaterialFunctionCall(Transition Phase=Multiply(A=Saturate(None=@n24), B=ComponentMask(None=@n60) {b=true}), Height Texture=NamedRerouteUsage()) {materialFunction=&HeightLerp}
  PinBase!(Lerp Alpha No Contrast=@n61.Lerp Alpha No Contrast) {reroutePins [Lerp Alpha No Contrast {expression=&self:MaterialExpressionReroute_28}]}
  MaterialExpressionReroute_11 = Reroute(None=@n21.Lerp Alpha No Contrast)
  ScalarParameter() {parameterName="Grass "}
  MaterialExpressionReroute_4 = Reroute(None=@n3)
  SphereMask(
   A=WorldPosition().XYZ
   B=VectorParameter() {parameterName=Sand Center, defaultValue=<rgba -5536 -89696 1280 1>}.RGB
   Radius=ScalarParameter() {defaultValue=5500, parameterName=Cave Center Sand Radius}
   Hardness=@MaterialExpressionReroute_4.Hardness
  )
  n62 = RuntimeVirtualTextureSampleParameter() {
   group=RVT
   materialType=BaseColor_Normal_Specular_YCoCg
   parameterName=RVT
   virtualTexture=/Game/Athena/Apollo/Maps/Landscape/VT/RVT_Apollo_Landscape
  }
  n63 = SetMaterialAttributes!(MaterialAttributes, Base Color=@n62.BaseColor, Specular=@n62.BaseColor, Roughness=@n62.BaseColor) {attributeSetTypes [#1, #3, #2]}
  n64 = NamedRerouteUsage()
  MaterialExpressionReroute_28 = Reroute(None=@n61.Lerp Alpha No Contrast)
  n65 = BlendMaterialAttributes(A=NamedRerouteUsage(), B=NamedRerouteUsage(), Alpha=@MaterialExpressionReroute_28.Lerp Alpha No Contrast)
  n66 = NamedRerouteUsage()
  n67 = NamedRerouteUsage()
  MaterialExpressionReroute_9 = Reroute(None=@n59.Alpha)
  n68 = BlendMaterialAttributes(
   A=NamedRerouteDeclaration(None=BlendMaterialAttributes(A=@n66, B=@n67, Alpha=@MaterialExpressionReroute_9.Alpha)) {name=Sand Rocks Mix, nodeColor=<rgba 0 1 0.1562 1>}
   B=NamedRerouteUsage()
   Alpha=Multiply(A=NamedRerouteUsage(), B=NamedRerouteUsage())
  )
  n69 = GetMaterialAttributes(None=NamedRerouteDeclaration(None=StaticSwitchParameter(True=@n65, False=@n68) {group=Grass, parameterName=Grass Top}) {name=Sand As is, nodeColor=<rgba 0.5937 1 0 1>}) {attributeGetTypes [#1]}
  n70 = SetMaterialAttributes!(
   MaterialAttributes=@n69.MaterialAttributes
   Base Color=Multiply(A=@n69.Base Color, B=VectorParameter() {group=Wet Walls, parameterName="Wet Color (Epic)", defaultValue=<rgba 0.8281 0.3828 0.7226 1>}.RGB)
   Roughness=Constant() {r=-0.55}
   Specular=Constant() {r=0.5}
  ) {attributeSetTypes [#1, #2, #3]}
  n71 = ComponentMask(None=NamedRerouteUsage()) {r=true}
  n72 = Max(
   A=SphereMask(A=WorldPosition().XYZ, B=Constant3Vector() {constant=<rgba -7988 -85296 2000 1>}, Radius=Constant() {r=3700}, Hardness=Constant() {r=0.35})
   B=SphereMask(A=WorldPosition().XYZ, B=Constant3Vector() {constant=<rgba -5680 -82556 2740 1>}, Radius=Constant() {r=3500}, Hardness=Constant() {r=0.5})
  )
  n73 = SphereMask(A=WorldPosition().XYZ, B=Constant3Vector() {constant=<rgba -5984 -86224 1592 1>}, Radius=Constant() {r=1000}, Hardness=Constant() {r=0.4})
  n74 = WorldPosition()
  n75 = Constant3Vector() {constant=<rgba -8936 -86888 1592 1>}
  n76 = Constant() {r=1500}
  n77 = Constant() {r=0.5}
  n78 = Multiply(A=NamedRerouteUsage(), B=Max(A=Max(A=@n72, B=@n73), B=SphereMask(A=@n74.XYZ, B=@n75, Radius=@n76, Hardness=@n77)))
  n79 = Constant() {r=1}
  n80 = Constant() {r=0.7}
  n81 = Saturate(None=ComponentMask(None=VertexNormalWS()) {b=true})
  n82 = Saturate(None=Multiply(A=Multiply(A=@n71, B=@n78), B=LinearInterpolate(A=@n79, B=@n80, Alpha=@n81)))
  n83 = MaterialFunctionCall!(Material=QualitySwitch(Default=@n64, Low=@n64, High=@n64, Medium=@n64, Epic=BlendMaterialAttributes(A=@n64, B=@n70, Alpha=@n82))) {materialFunction=/Game/Athena/Apollo/Maps/Landscape/VT/MaterialFunctions/TransformTangentToWorld}
  n84 = RuntimeVirtualTextureSampleParameter() {group=RVT, materialType=WorldHeight, parameterName=RVTH, virtualTexture=/Game/Athena/Apollo/Maps/Landscape/VT/RVT_Apollo_Landscape_Height}
  n85 = Subtract(A=ComponentMask(None=WorldPosition().XYZ) {b=true}, B=@n84.WorldHeight)
  n86 = ScalarParameter() {defaultValue=15, group=RVT, parameterName=Blend RVT N}
  n87 = BlendMaterialAttributes(
   A=SetMaterialAttributes!(MaterialAttributes, Normal=@n62.Normal) {attributeSetTypes [#4]}
   B=@n83.Result
   Alpha=Saturate(None=Divide(A=@n85, B=@n86))
  ) {vertexAttributeBlendType=UseA}
  n88 = ScalarParameter() {defaultValue=25, group=RVT, parameterName=Blend RVT}
  n89 = ScalarParameter() {defaultValue=55, group=RVT, parameterName=Blend RVT Up}
  n90 = VertexNormalWS()
  n91 = NamedRerouteDeclaration(None=MaterialFunctionCall(In=ComponentMask(None=@n90) {b=true}, Contrast=Constant() {r=1}) {materialFunction=&CheapContrast}.Result) {name=Up Mask Contr 1, nodeColor=<rgba 0.8281 0 1 1>}
  n92 = NamedRerouteUsage()
  n93 = LinearInterpolate(A=NamedRerouteUsage(), B=NamedRerouteUsage(), Alpha=@n92)
  n94 = MaterialFunctionCall(
   Transition Phase=Divide(A=@n85, B=LinearInterpolate(A=@n88, B=@n89, Alpha=@n91))
   Height Texture=LinearInterpolate(A=Constant() {r=-1}, B=Constant() {r=1}, Alpha=ComponentMask(None=@n93) {r=true})
   Contrast=@n92
  ) {materialFunction=&HeightLerp}
  n95 = LinearInterpolate(A=@n94.Lerp Alpha No Contrast, B=@n94.Lerp Alpha No Contrast, Alpha=@n92)
  n96 = MaterialFunctionCall(
   NoRVT=@n83.Result
   RVT=MaterialProxyReplace(Realtime=BlendMaterialAttributes(A=@n63, B=@n87, Alpha=@n95) {vertexAttributeBlendType=UseA}, MaterialProxy=@n83.Result)
  ) {materialFunction=&FN_RVT_Switch_Material}
  StaticSwitchParameter(True=@n83.Result, False=ShadingPathSwitch(Default=@n96.Result, Deferred=@n96.Result, Forward=@n83.Result, Mobile=@n83.Result)) {parameterName=NoRVT}
 ]
}
