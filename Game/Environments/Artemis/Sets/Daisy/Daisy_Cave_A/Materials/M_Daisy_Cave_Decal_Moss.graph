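# Translucent deferred decal material that lays tintable moss, named for the Daisy cave environment set.
# A "WS Decal Tiling" switch picks world-aligned triplanar (default) or UV-tiled sampling of diffuse, normal and packed specular/metallic/roughness textures.
# Opacity multiplies a Decal Alpha mask by the diffuse texture (alpha when world-aligned), scaled by an Opacity parameter; a tint layer function colors the result.
# A spherical-gradient, erosion and decal-facing mask branch is present but not wired into opacity.
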Material /Game/Environments/Artemis/Sets/Daisy/Daisy_Cave_A/Materials/M_Daisy_Cave_Decal_Moss
graph {
 properties {
  bUseMaterialAttributes=true
  blendMode=BLEND_Translucent
  materialDomain=MD_DeferredDecal
  pixelDepthOffsetMode=PDOM_Legacy
  usedShadingModels=Default Lit
 }
 nodes [
  n1 = TextureObjectParameter() {parameterName=Decal, samplerType=SAMPLERTYPE_Masks, texture=/Game/Environments/Artemis/Sets/Terrain/Textures/T_Artemis_Frontier_Moss_01_D_03}
  n2 = TextureObjectParameter() {parameterName=Specular, samplerType=SAMPLERTYPE_Masks, texture=/Game/Environments/Artemis/Sets/Terrain/Textures/T_Artemis_Frontier_Moss_01_S}
  n3 = TextureObjectParameter() {parameterName=Normal, samplerType=SAMPLERTYPE_Normal, texture=/Game/Environments/Artemis/Sets/Terrain/Textures/T_Artemis_Frontier_Moss_01_N}
  n4 = MaterialFunctionCall!(
   Diffuse Texture Object=@n1
   Spec Texture Object=@n2
   TextureSize=ScalarParameter() {defaultValue=512, parameterName=WS Decal Texture Size}
   --------------------------
   Use seperate settings for Z projection
   Use seperate texture for Z projection
   Texture Size Z projection
   Z Plane Diffuse Texture Object
   Z Plane Spec Texture Object
   "--------------------------      "
   ProjectionTransitionContrast
   Use Texture Alpha Heightmap to Refine the Transistion
   HeightLerpTransitionContrast
   PickTheDiffuseHeightChannel
   TheDiffuseHeightChannel
   "--------------------------       "
   NormalTextureObject=@n3
   World Space Normals
   "--------------------------  "
   Export Float 4 Diffuse=StaticBool() {value=true}
   Export Float 4 Spec
   "--------------------------     "
   WorldPosition
   World Space Normal
  ) {materialFunction=/Engine/Functions/Engine_MaterialFunctions02/Texturing/WorldAlignedTextures_Complex}
  n5 = ComponentMask(None=@n4.XYZ Diffuse Texture) {b=true, g=true, r=true}
  n6 = Multiply(A=ScalarParameter() {defaultValue=1, parameterName=Decal Texture Size}, B=TextureCoordinate())
  n7 = TextureSample(UVs=@n6, Tex=@n1) {samplerType=SAMPLERTYPE_Masks, texture=&T_Artemis_Frontier_Moss_01_D_03}
  n8 = StaticSwitchParameter(
   True=@n4.XYZ Spec
   False=TextureSample(UVs=@n6, Tex=@n2) {samplerType=SAMPLERTYPE_Masks, texture=&T_Artemis_Frontier_Moss_01_D_03}.RGB
  ) {defaultValue=true, parameterName=WS Decal Tiling}
  n9 = TextureSampleParameter2D() {parameterName=Decal Alpha, samplerType=SAMPLERTYPE_LinearColor, texture=/Game/Global/Textures/Default/Masks_and_Gradients/RevealMask}
  n10 = ComponentMask(None=@n4.XYZ Diffuse Texture) {a=true}
  n11 = Multiply(A=@n9.RGB, B=StaticSwitchParameter(True=@n10, False=@n7.RGB) {defaultValue=true, parameterName=WS Decal Tiling})
  n12 = ScalarParameter() {defaultValue=1, parameterName=Opacity, sliderMax=1}
  n13 = TextureSample(UVs=@n6, Tex=@n3) {samplerType=SAMPLERTYPE_Masks, texture=&T_Artemis_Frontier_Moss_01_D_03}
  MaterialFunctionCall!(
   Material=MakeMaterialAttributes!(
    BaseColor=StaticSwitchParameter(True=@n5, False=@n7.RGB) {defaultValue=true, parameterName=WS Decal Tiling}
    Metallic=ComponentMask(None=@n8) {g=true}
    Specular=ComponentMask(None=@n8) {r=true}
    Roughness=ComponentMask(None=@n8) {b=true}
    Anisotropy
    EmissiveColor
    Opacity=LinearInterpolate(B=@n11, Alpha=@n12)
    OpacityMask
    Normal=StaticSwitchParameter(True=@n4.Normal XYZ Texture, False=@n13.RGB) {defaultValue=true, parameterName=WS Decal Tiling}
    Tangent
    WorldPositionOffset
    SubsurfaceColor
    ClearCoat
    ClearCoatRoughness
    AmbientOcclusion
    Refraction
    CustomizedUVs_0
    CustomizedUVs_1
    CustomizedUVs_2
    CustomizedUVs_3
    CustomizedUVs_4
    CustomizedUVs_5
    CustomizedUVs_6
    CustomizedUVs_7
    PixelDepthOffset
    ShadingModel
    Displacement
   )
   Tint=VectorParameter() {parameterName=Moss Decal Color, defaultValue=<rgba 0.65 0.325 0.4875 1>}.RGB
   TintMask
  ) {materialFunction=/Engine/Functions/MaterialLayerFunctions/MatLayerBlend_Tint}
  n14 = SphereMask(
   A=TextureCoordinate()
   B=AppendVector(
    A=ScalarParameter() {defaultValue=0.5, parameterName=SphericalGradient_Center_U}
    B=ScalarParameter() {defaultValue=0.5, parameterName=SphericalGradient_Center_V}
   )
   Radius=ScalarParameter() {defaultValue=0.5, parameterName=SphericalGradientRadius}
   Hardness=ScalarParameter() {defaultValue=0.25, parameterName=SphericalGradientHardness}
  )
  Multiply(
   A=Subtract(A=Multiply(A=@n14, B=@n10), B=ScalarParameter() {parameterName=Erosion})
   B=MaterialFunctionCall() {materialFunction=/Game/Weapons/Prototype/SprayGun/Materials/MF_DecalFacingMask}.Result
  )
 ]
}
